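; Weapon component for dual-magazine weapons that switches the owning weapon's next reload between fast and slow.
; On BeginPlay it caches the owning FortWeapon and async loads the override fire mode data.
; SetReloadType sets the rep-notify IsNextReloadFast, whose OnRep calls ApplyReloadType.
; Fast clears the fire mode override; slow applies it and allows reloading when full, with both broadcasting SwappedReloadSpeed.
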(event EventBeginPlay
  (bind _override_fire_mode_data (Variables|Default|GetOverride%20Fire%20Mode%20Data))
  ;[0.09387 1 0.08132] Cache the weapon associated with this component and load the fire mode asset
  (bind _asfort_weapon (Utilities|Casting|CastToFortWeapon (Components|GetOwner)))
  (Variables|Default|SetWeapon _asfort_weapon)
  (if (Utilities|IsValidSoftObjectReference _override_fire_mode_data)
    (bind _object (Utilities|AsyncLoadAsset _override_fire_mode_data)
      (:Completed
        (Variables|Default|SetLoadedFireMode _object)))))

(fn ApplyReloadType
  (params (Fast :type Boolean))
  (bind _weapon (Variables|Default|GetWeapon))
  (bind _weapon_2 (Variables|Default|GetWeapon))
  ;[1 1 0] Apply the next reload (fast or slow)
  (Utilities|IsValid _weapon
    (:Is%20Valid
      (Utilities|FlowControl|Sequence
        (:then_0
          ;;[0 1 1] Update the override fire mode data
          (if (Variables|GetFast)
            (FortWeapon|ClearOverrideFireModeData _weapon_2)
            (else
              (FortWeapon|SetOverrideFireModeData _weapon_2 (Variables|Default|GetOverride%20Fire%20Mode%20Data)))))
        (:then_1
          (FortWeapon|SetAllowReloadwhenFull _weapon (not Fast))
          (Default|CallSwappedReloadSpeed Fast))))))

(fn OnRep_IsNextReloadFast
  (CallFunction|ApplyReloadType (Variables|Default|GetIsNextReloadFast)))

(fn SetReloadType
  (params (bFast :type Boolean))
  (bind _output_get (Variables|Default|SetIsNextReloadFast bFast))
  (return _output_get))
